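# Preview physics asset for the small RoseBro character, part of the ChimneyStack large male player body cosmetic.
# One capsule body covers each major bone from pelvis through spine, head, arms, hands, thighs and calves, with an oversized head capsule and no foot bodies.
# Constraints form a tree rooted at the pelvis, with arms and head branching from spine_05, and every joint shares the same uniform swing and twist limit angles.
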PhysicsAsset /BRCosmetics/Characters/Player/Male/Large/Bodies/M_LRG_ChimneyStack/RoseBros/M_MED_RoseBro_Small/PA_RoseBro_Small_Preview
bodies {
 calf_l [{shapeType=Capsule, length=11.1202, radius=8.007, center=<-7.331 2.4767 0.7646>, rotation=<rot 0 14.7269 -87.2876>}]
 calf_r [{shapeType=Capsule, length=11.1446, radius=8.0076, center=<7.3301 -2.4684 -0.7246>, rotation=<rot 0 14.9097 -87.6398>}]
 hand_l [{shapeType=Capsule, length=9.8914, radius=7.0138, center=<6.1827 -0.9741 0.0161>, rotation=<rot 0 81.2107 -79.7172>}]
 hand_r [{shapeType=Capsule, length=9.8909, radius=7.0191, center=<-6.1822 0.9725 -0.022>, rotation=<rot 0 81.2425 -79.6516>}]
 head [{shapeType=Capsule, length=36.3506, radius=27.9454, center=<16.3066 3.7212 -1.847>, rotation=<rot 4.6451 -90 -8.154>}]
 lowerarm_l [{shapeType=Capsule, length=9.139, radius=6.4996, center=<5.6938 0.1655 0.102>, rotation=<rot 0 -88.1236 -75.8167>}]
 lowerarm_r [{shapeType=Capsule, length=9.1345, radius=6.4965, center=<-5.6904 -0.1608 -0.0929>, rotation=<rot 0 -88.1788 -75.7598>}]
 pelvis [{shapeType=Capsule, length=18.8654, radius=16.0313, center=<2.8243 -1.4595 0.0695>, rotation=<rot 0.7685 -90 0.1154>}]
 spine_02 [{shapeType=Capsule, length=12.0746, radius=11.2305, center=<7.8992 2.1083 0.1793>, rotation=<rot 0 -0.646 -88.8655>}]
 spine_05 [{shapeType=Capsule, length=19.9649, radius=11.7933, center=<2.3264 2.2349 0.0002>, rotation=<rot -0.0507 -90 -0.0022>}]
 thigh_l [{shapeType=Capsule, length=10.3334, radius=10.2653, center=<-8.3688 1.3587 -0.1393>, rotation=<rot -25.0751 -90 -116.4621>}]
 thigh_r [{shapeType=Capsule, length=10.3974, radius=10.2753, center=<8.3353 -1.3587 0.1246>, rotation=<rot -24.9588 -90 -116.5622>}]
 upperarm_l [{shapeType=Capsule, length=9.8002, radius=6.3557, center=<6.7531 -0.2774 0.5468>, rotation=<rot 0 -92.2654 -84.3512>}]
 upperarm_r [{shapeType=Capsule, length=9.7877, radius=6.3577, center=<-6.7606 0.2752 -0.5353>, rotation=<rot 0 -92.2239 -84.4099>}]
}
constraints [
 {bone1Name=thigh_l, bone2Name=pelvis, swing1LimitDegrees=45, swing2LimitDegrees=45, twistLimitDegrees=45}
 {bone1Name=calf_l, bone2Name=thigh_l, swing1LimitDegrees=45, swing2LimitDegrees=45, twistLimitDegrees=45}
 {bone1Name=thigh_r, bone2Name=pelvis, swing1LimitDegrees=45, swing2LimitDegrees=45, twistLimitDegrees=45}
 {bone1Name=calf_r, bone2Name=thigh_r, swing1LimitDegrees=45, swing2LimitDegrees=45, twistLimitDegrees=45}
 {bone1Name=spine_02, bone2Name=pelvis, swing1LimitDegrees=45, swing2LimitDegrees=45, twistLimitDegrees=45}
 {bone1Name=spine_05, bone2Name=spine_02, swing1LimitDegrees=45, swing2LimitDegrees=45, twistLimitDegrees=45}
 {bone1Name=head, bone2Name=spine_05, swing1LimitDegrees=45, swing2LimitDegrees=45, twistLimitDegrees=45}
 {bone1Name=upperarm_r, bone2Name=spine_05, swing1LimitDegrees=45, swing2LimitDegrees=45, twistLimitDegrees=45}
 {bone1Name=lowerarm_r, bone2Name=upperarm_r, swing1LimitDegrees=45, swing2LimitDegrees=45, twistLimitDegrees=45}
 {bone1Name=hand_r, bone2Name=lowerarm_r, swing1LimitDegrees=45, swing2LimitDegrees=45, twistLimitDegrees=45}
 {bone1Name=upperarm_l, bone2Name=spine_05, swing1LimitDegrees=45, swing2LimitDegrees=45, twistLimitDegrees=45}
 {bone1Name=lowerarm_l, bone2Name=upperarm_l, swing1LimitDegrees=45, swing2LimitDegrees=45, twistLimitDegrees=45}
 {bone1Name=hand_l, bone2Name=lowerarm_l, swing1LimitDegrees=45, swing2LimitDegrees=45, twistLimitDegrees=45}
]
